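# Translucent UI material that draws the sawblade gun's reticle as a patterned donut ring with a progress parameter and tintable color.
# Rotated, center-scaled UVs become radial coordinates, wrapping the pattern texture into segments, spun manually or by time.
# A circle-function outline shapes the donut, while a smoothstepped angle gradient minus Progress forms a hard progress mask.
# Static switches toggle pattern and direction inversion, noise-texture masking and cosine-based edge fading.
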Material /SawbladeGun/Items/SawbladeGun/UI/M_UI_SawReticle
graph {
 properties {blendMode=BLEND_Translucent, materialDomain=MD_UI, pixelDepthOffsetMode=PDOM_Legacy, usedShadingModels=Default Lit}
 nodes [
  n1 = NamedRerouteUsage()
  n2 = TextureObjectParameter() {
   group=0- Visual
   parameterName=Pattern Texture
   samplerType=SAMPLERTYPE_Alpha
   sortPriority=0
   texture=&/UI_SawbladeGun_Pattern
  }
  n3 = TextureSample(UVs=NamedRerouteUsage(), Tex=@n2) {samplerSource=SSM_Wrap_WorldGroupSettings}
  n4 = StaticSwitchParameter(True=@n3.R, False=@n3.R) {defaultValue=true, group=0- Visual, parameterName=Use Pattern Red Channel, sortPriority=0}
  n5 = LinearInterpolate(B=LinearInterpolate(B=OneMinus(None=@n4), Alpha=NamedRerouteUsage()), Alpha=NamedRerouteUsage())
  n6 = StaticSwitchParameter(True=Saturate(None=Multiply(A=@n1, B=@n5)), False=@n5) {group=4- Texture Mask, parameterName=Use Texture Mask, sortPriority=0}
  n7 = NamedRerouteUsage()
  Multiply(
   A=StaticSwitchParameter(True=LinearInterpolate(B=@n6, Alpha=@n7), False=@n6) {group=5- Edge Mask, parameterName=Fade Out Edges, sortPriority=0}
   B=ScalarParameter() {defaultValue=1, group=0- Visual, parameterName=Opacity, sliderMax=1}
  )
  NamedRerouteDeclaration(None=MaterialFunctionCall(
    Size=ScalarParameter() {defaultValue=0.54, group=2- Donut Mask, parameterName=Donut Size, sortPriority=2}
    OutlineThickness=ScalarParameter() {defaultValue=0.13, group=2- Donut Mask, parameterName=Donut Outline Thickness, sortPriority=5}
    GlowMin=ScalarParameter() {group=2- Donut Mask, parameterName=Donut Glow Min, sortPriority=3}
    GlowMax=ScalarParameter() {group=2- Donut Mask, parameterName=Donut Glow Max, sortPriority=4}
   ) {materialFunction=&MF_UI_Circle}.Outline) {name=Donut Mask, nodeColor=<rgba 0 0 0 1>}
  n8 = MaterialFunctionCall!(UVs, Rotation=ScalarParameter() {defaultValue=-0.25, group=2- Donut Mask, parameterName=Donut Angle, sliderMax=1, sortPriority=0}) {materialFunction=/FortUILibrary/MaterialFunctions/Gradients/MF_UI_AngleGradient}
  n9 = SmoothStep(
   Min=ScalarParameter() {defaultValue=0.75, group=2- Donut Mask, parameterName=Start Angle, sliderMax=1, sortPriority=1}
   Value=StaticSwitchParameter(True=OneMinus(None=@n8.Result), False=@n8.Result) {defaultValue=true, group=3- Progress, parameterName=Inverse Progress}
  )
  n10 = Subtract(A=@n9, B=ScalarParameter() {group=3- Progress, parameterName=Progress, sliderMax=1})
  NamedRerouteDeclaration(None=Saturate(None=Ceil(None=@n10))) {name=Progress Mask, nodeColor=<rgba 0 0 0 1>}
  n11 = TextureObjectParameter() {group=4- Texture Mask, parameterName=Mask Texture, sortPriority=1, texture=&AtmosphericCloudNoise01}
  n12 = TextureSample(
   UVs=Multiply(A=TextureCoordinate(), B=ScalarParameter() {defaultValue=1, group=4- Texture Mask, parameterName=Mask Scale}) {constB=2}
   Tex=@n11
  ) {samplerSource=SSM_Wrap_WorldGroupSettings}
  n13 = OneMinus(None=ScalarParameter() {group=4- Texture Mask, parameterName=Alpha Threshold})
  NamedRerouteDeclaration(None=Saturate(None=Power(Base=@n12.R, Exp=@n13))) {name=Texture Mask, nodeColor=<rgba 0 0 0 1>}
  n14 = OneMinus(None=Cosine(None=@n9))
  n15 = ScalarParameter() {defaultValue=0.5, group=5- Edge Mask, parameterName=Edge Mask Falloff, sortPriority=1}
  NamedRerouteDeclaration(None=Saturate(None=Power(Base=@n14, Exp=@n15) {constExponent=0.5})) {name=Edge Mask, nodeColor=<rgba 0 0 0 1>}
  n16 = ScalarParameter() {defaultValue=18, group=1- Pattern Settings, parameterName=Segment Scale, sortPriority=2}
  n17 = ScalarParameter() {defaultValue=1, group=1- Pattern Settings, parameterName=Layer Scale, sortPriority=3}
  n18 = TextureCoordinate()
  n19 = MaterialFunctionCall(Float2=@n18) {materialFunction=&BreakOutFloat2Components}
  n20 = AppendVector(A=@n19.R, B=@n19.R)
  n21 = StaticSwitchParameter(
   True=ScalarParameter() {group=1- Pattern Settings, parameterName=Rotation, sortPriority=1}
   False=Multiply(A=Time(), B=ScalarParameter() {defaultValue=0.6, parameterName=Speed}) {constB=0.6}
  ) {defaultValue=true, group=1- Pattern Settings, parameterName=Manual Rotation, sortPriority=0}
  n22 = OneMinus(None=@n21)
  n23 = MaterialFunctionCall(
   UVs=MaterialFunctionCall(
    UVs=StaticSwitchParameter(True=@n20, False=@n18) {defaultValue=true, group=0- Visual, parameterName=Inverse Pattern, sortPriority=1}
    "Rotation Angle (0-1)"=StaticSwitchParameter(True=@n22, False=@n21) {group=1- Pattern Settings, parameterName=Inverse Rotation, sortPriority=1}
   ) {materialFunction=&CustomRotator}.Rotated Values
   Texture Scale=ScalarParameter() {defaultValue=0.125, group=1- Pattern Settings, parameterName=Scale, sortPriority=4}
  ) {materialFunction=&ScaleUVsByCenter}
  NamedRerouteDeclaration(None=Multiply(
    A=AppendVector(A=@n16, B=@n17)
    B=MaterialFunctionCall!(Vector or UVs=@n23.UVs) {materialFunction=/FortUILibrary/MaterialFunctions/Basics/MF_UI_CheaperVectorToRadial}.Radial Coordinates
   ) {constB=4}) {name=UVs, nodeColor=<rgba 0 0 0 1>}
  OneMinus(None=@n19.R)
  VectorParameter() {group=0- Visual, parameterName=Color, sortPriority=2, defaultValue=<rgba 1 1 1 1>}
 ]
}
