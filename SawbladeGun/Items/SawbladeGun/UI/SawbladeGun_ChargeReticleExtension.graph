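# Charged-weapon reticle extension widget for the Sawblade Gun, showing charge as a radial progress bar and a spinning sawblade.
# Construct starts a looping material-rotation animation and applies LastCharge to both visuals.
# Changed charge updates set the progress material's Percent and scale the blade's spin speed, opacity and color.
# Comparison with LastCharge fades the center reticle out when charge rises and in when it falls, using mutually resetting DoOnce nodes.
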WidgetBlueprint /SawbladeGun/Items/SawbladeGun/UI/SawbladeGun_ChargeReticleExtension
animations [
 FadeInCenterReticle {
  binding_count=1
  length_seconds=0.5
  object=&self:FadeInCenterReticle
  sequence {bindings [CenterReticle {
     id=#1
     possessed_class=Image
     tracks [{
       type=MovieSceneFloatTrack
       display_name=Render Opacity
       sections [{type=MovieSceneFloatSection, end_frame=10, start_frame=0, channels [None {type=MovieSceneScriptingFloatChannel, keys [[0,0],[10,1]]}]}]
      }]
    }]}
 }
 FadeOutCenterReticle {
  binding_count=1
  length_seconds=0.5
  object=&self:FadeOutCenterReticle
  sequence {bindings [CenterReticle {
     id=#2
     possessed_class=Image
     tracks [{
       type=MovieSceneFloatTrack
       display_name=Render Opacity
       sections [{type=MovieSceneFloatSection, end_frame=10, start_frame=0, channels [None {type=MovieSceneScriptingFloatChannel, keys [[0,1],[10,0]]}]}]
      }]
    }]}
 }
 SawRotation {
  binding_count=1
  length_seconds=1
  object=&self:SawRotation
  sequence {bindings [Sawblade {
     id=#3
     possessed_class=Image
     tracks [{
       type=MovieSceneWidgetMaterialTrack
       sections [{
         type=MovieSceneParameterSection
         end_frame=20
         start_frame=0
         channels [Rotation {type=MovieSceneScriptingFloatChannel, keys [[0,0],[20,1]]}]
        }]
      }]
    }]}
 }
]
graphs [
 EventGraph {nodes [
   Event Construct()
   > PlayAnimation(InAnimation=Get SawRotation() -> (SawRotation:Widget Animation Object Reference), NumLoopsToPlay=0)
   n1 = Get LastCharge() -> (LastCharge:double)
   > UpdateProgressVisual!(execute, self, ChargePercent:double=@n1) -> (then)
   > UpdateSawbladeVisual!(execute, self, ChargePercent:double=@n1) -> (then)
   n2 = Event OnUpdateChargeDisplay!() -> (OutputDelegate, then, ChargePercent:float) {is=title K2Node_Event}
   > Parent: OnUpdateChargeDisplay!(execute, ChargePercent:float=@n2.ChargePercent) -> (then) {is=title K2Node_CallParentFunction}
   > Branch(Condition="Not Equal (Float)"(A=Get LastCharge(), B=@n2.ChargePercent))
   > UpdateProgressVisual(ChargePercent=@n2.ChargePercent)
   > UpdateSawbladeVisual(ChargePercent=@n2.ChargePercent)
   > n3 = Sequence()
   > n4 = Branch(Condition="float > float"(A=Get LastCharge(), B=@n2.ChargePercent))
   > Do Once(Reset=@n6.then)
   > n5 = PlayAnimation(InAnimation=Get FadeInCenterReticle() -> (FadeInCenterReticle:Widget Animation Object Reference))
   Do Once(execute=@n4.else, Reset=@n5.then)
   > n6 = PlayAnimation(InAnimation=Get FadeOutCenterReticle() -> (FadeOutCenterReticle:Widget Animation Object Reference))
   Set LastCharge(execute=@n3.then_1, LastCharge:double=@n2.ChargePercent) -> (Output_Get:double)
  ]}
 UpdateProgressVisual {nodes [
   n1 = UpdateProgressVisual!() -> (then, ChargePercent:double=~) {is=title K2Node_FunctionEntry}
   > n2 = GetDynamicMaterial(self=Get Image_Progress() -> (Image_Progress:Image Object Reference))
   > Is Valid(InputObject=@n2)
   > SetScalarParameterValue(self=@n2, ParameterName=Percent, Value=@n1)
  ]}
 UpdateSawbladeVisual {nodes [
   n1 = UpdateSawbladeVisual!() -> (then, ChargePercent:double=~) {is=title K2Node_FunctionEntry}
   > SetPlaybackSpeed!(
    execute
    self:User Widget Object Reference
    InAnimation:Widget Animation Object Reference=Get SawRotation()
    PlaybackSpeed:float=Lerp(A=0, B=0.75, Alpha=@n1)
   ) -> (then)
   > n2 = GetDynamicMaterial(self=Get Sawblade() -> (Sawblade:Image Object Reference))
   > Is Valid(InputObject=@n2)
   > SetScalarParameterValue(self=@n2, ParameterName=Opacity, Value=Lerp(A=0, B=1, Alpha=@n1))
   > SetScalarParameterValue(self=@n2, ParameterName=Color, Value=Lerp(A=0.8, B=1, Alpha=@n1))
  ]}
]
variables [LastCharge {type=number}]
widgets {
 info {
  inheritedWidgetCount=1
  parentClass=/Script/FortniteUI.FortChargedWeaponReticleExtensionWidget
  rootWidgetClass=&Overlay
  widgetCount=8
 }
 widgets [
  Overlay_0 {widgetClassPath=&Overlay, widget=Overlay_0, properties {visibility=HitTestInvisible}}
  CommonBorder {
   widgetClassPath=&CommonBorder
   parent=Overlay_0
   widget=CommonBorder
   properties {padding=<ltrb 0 0 0 150>}
   slot {class=OverlaySlot, properties {horizontalAlignment=HAlign_Center, verticalAlignment=VAlign_Center}}
  }
  Image_Progress {
   widgetClassPath=&Image
   bIsVariable=true
   parent=CommonBorder
   widget=Image_Progress
   properties {
    visibility=HitTestInvisible
    brush {
     bIsDynamicallyLoaded=false
     drawAs=Image
     imageType=NoImage
     mirroring=NoMirror
     resourceName=None
     resourceObject=/Game/Athena/HUD/Materials/M_UI_RadialProgress_Ammo
     tiling=NoTile
     imageSize=<64 64>
     margin=<ltrb 0 0 0 0>
     outlineSettings {
      bUseBrushTransparency=false
      roundingType=HalfHeightRadius
      width=0
      color {colorUseRule=UseColor_Specified, specifiedColor=<rgba 0 0 0 0>}
      cornerRadii=<0 0 0 0>
     }
     tintColor {colorUseRule=UseColor_Specified, specifiedColor=<rgba 1 1 1 1>}
     uVRegion {bIsValid=false, max=<0 0>, min=<0 0>}
    }
    colorAndOpacity=<rgba 1 1 1 0.8>
    renderTransform {angle=-45, scale=<-1 1>, shear=<0 0>, translation=<0 0>}
   }
   slot {class=BorderSlot, properties {padding=<ltrb 0 0 0 150>}}
  }
  CenterReticle {
   widgetClassPath=&Image
   parent=Overlay_0
   widget=CenterReticle
   properties {brush {
     bIsDynamicallyLoaded=false
     drawAs=Image
     imageType=NoImage
     mirroring=NoMirror
     resourceName=None
     resourceObject=&/HUD_Xhair_Sawblade_comp
     tiling=NoTile
     imageSize=<64 64>
     margin=<ltrb 0 0 0 0>
     outlineSettings {
      bUseBrushTransparency=false
      roundingType=HalfHeightRadius
      width=0
      color {colorUseRule=UseColor_Specified, specifiedColor=<rgba 0 0 0 0>}
      cornerRadii=<0 0 0 0>
     }
     tintColor {colorUseRule=UseColor_Specified, specifiedColor=<rgba 1 1 1 1>}
     uVRegion {bIsValid=false, max=<0 0>, min=<0 0>}
    }}
   slot {class=OverlaySlot, properties {horizontalAlignment=HAlign_Center, verticalAlignment=VAlign_Center}}
  }
  Overlay_1 {
   widgetClassPath=&Overlay
   parent=Overlay_0
   widget=Overlay_1
   properties {renderTransform {angle=0, scale=<1 1>, shear=<0 0>, translation=<0 -27>}}
   slot {class=OverlaySlot, properties {horizontalAlignment=HAlign_Center, verticalAlignment=VAlign_Center}}
  }
  Sawblade {
   widgetClassPath=&Image
   bIsVariable=true
   parent=Overlay_1
   widget=Sawblade
   properties {brush {
     bIsDynamicallyLoaded=false
     drawAs=Image
     imageType=NoImage
     mirroring=NoMirror
     resourceName=None
     resourceObject=&/MI_UI_SawReticle
     tiling=NoTile
     imageSize=<32 32>
     margin=<ltrb 0 0 0 0>
     outlineSettings {
      bUseBrushTransparency=false
      roundingType=HalfHeightRadius
      width=0
      color {colorUseRule=UseColor_Specified, specifiedColor=<rgba 0 0 0 0>}
      cornerRadii=<0 0 0 0>
     }
     tintColor {colorUseRule=UseColor_Specified, specifiedColor=<rgba 1 1 1 1>}
     uVRegion {bIsValid=false, max=<0 0>, min=<0 0>}
    }}
   slot {class=OverlaySlot, properties {horizontalAlignment=HAlign_Fill, verticalAlignment=VAlign_Fill}}
  }
  Spacer_121 {
   widgetClassPath=&Spacer
   parent=Overlay_1
   widget=Spacer_121
   properties {size=<212 212>}
   slot {class=OverlaySlot, properties {horizontalAlignment=HAlign_Fill, verticalAlignment=VAlign_Fill}}
  }
  TutorialHighlight {widgetClassPath=&TutorialHighlight}
 ]
}
